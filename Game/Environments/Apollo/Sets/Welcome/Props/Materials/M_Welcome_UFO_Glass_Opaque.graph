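# Opaque Default Lit material that fakes cracked, dirty glass for Welcome set UFO props.
# One attribute set blends two base colors by crack mask and Fresnel rim, mixes in dirt via height-lerped radial ring, and uses cracks for roughness and normal.
# It also drives hit-and-shrink WPO through Fortnite_HitEffects with custom primitive data.
# A separate, unchained attribute set builds metallic thin-film tinting from Fresnel gradient lookup and atmospheric-light reflection highlights.
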Material /Game/Environments/Apollo/Sets/Welcome/Props/Materials/M_Welcome_UFO_Glass_Opaque
graph {
 properties {bUseMaterialAttributes=true, pixelDepthOffsetMode=PDOM_Legacy, refractionMethod=RM_IndexOfRefraction, usedShadingModels=Default Lit}
 nodes [
  n1 = Multiply(A=VectorParameter() {parameterName=DirtColor, defaultValue=<rgba 0.1771 0.1663 0.1388 0>}.RGB, B=Max(B=Constant() {r=0.35}))
  Add(A=Multiply() {constB=0.25}, B=@n1)
  n2 = DotProduct(A=AtmosphericLightVector(), B=ReflectionVectorWS())
  n3 = Power(Base=@n2, Exp=Constant() {r=512})
  n4 = Power(Base=@n2, Exp=Constant() {r=8})
  Add(A=Multiply(A=Constant() {r=155}, B=Saturate(None=@n3)), B=Multiply(A=Saturate(None=@n4), B=Constant() {r=0.4}))
  n5 = TextureCoordinate()
  n6 = ScalarParameter() {defaultValue=0.7102, parameterName=ScaleTiledNormal}
  n7 = MaterialFunctionCall(
   Normal=TextureSample(UVs=Multiply(A=@n5, B=@n6)) {samplerType=SAMPLERTYPE_Normal, texture=/Game/Gadgets/Assets/Telephone_Booth/Textures/T_BlindsNormalNoise_N}.RGB
   Flatness=ScalarParameter() {defaultValue=0.5, parameterName=FlattenNormal}
  ) {materialFunction=&FlattenNormal}
  n8 = StaticBoolParameter() {defaultValue=true, parameterName=UseNormalTexture}
  n9 = Fresnel(
   ExponentIn=ScalarParameter() {defaultValue=2, parameterName=Opacity-Fresnel Exponent}
   BaseReflectFractionIn=ScalarParameter() {defaultValue=0.5, parameterName=BaseTransparency}
   Normal=StaticSwitch(True=Transform(None=@n7.Result), False=VertexNormalWS(), Value=@n8)
  )
  n10 = Multiply(A=Fresnel() {baseReflectFraction=0, exponent=12}, B=ScalarParameter() {defaultValue=3, parameterName=GlassThickness})
  Reroute(None=Add(A=@n9, B=@n10))
  n11 = TextureSample(UVs=Multiply(A=TextureCoordinate(), B=ScalarParameter() {defaultValue=3, parameterName=DirtTile})) {samplerType=SAMPLERTYPE_Masks, texture=/Game/Athena/WinterArt/Textures/T_DirtLeaks_M}
  n12 = Constant() {r=0.8}
  n13 = TextureSample(UVs=@n9) {samplerSource=SSM_Clamp_WorldGroupSettings, texture=/Game/Global/Textures/Default/Masks_and_Gradients/3pxRGBImage}
  n14 = Multiply(A=Constant() {r=4}, B=@n13.RGB)
  n15 = DotProduct(A=AtmosphericLightVector(), B=ReflectionVectorWS())
  n16 = Constant() {r=55}
  n17 = Saturate(None=Multiply(A=Power(Base=@n15, Exp=@n16), B=Constant() {r=512}))
  n18 = VectorParameter() {parameterName=Base Color, defaultValue=<rgba 0.849 0.8087 0.7053 0>}
  n19 = LinearInterpolate(
   A=Multiply(A=@n18.RGB, B=@n18.RGB)
   B=Multiply(A=@n18.RGB, B=Multiply(A=@n18.RGB, B=@n18.RGB)) {constB=0.5}
   Alpha=MaterialFunctionCall(In=Saturate(None=@n10), Contrast=Constant() {r=3}) {materialFunction=&CheapContrast}.Result
  ) {constB=0}
  n20 = Multiply(
   A=ScalarParameter() {defaultValue=0.35, group=Anodizing, parameterName=Film, sortPriority=1}
   B=Multiply(A=LinearInterpolate(A=@n13.RGB, B=@n14, Alpha=@n17), B=@n19)
  )
  SetMaterialAttributes!(
   MaterialAttributes
   Base Color=Add(A=Max(A=@n11.B, B=@n12), B=Add(A=@n20, B=@n19))
   Metallic=Constant() {r=1}
   Specular
   Roughness
   Emissive Color
   Opacity
   Opacity Mask
   Normal=StaticSwitch(True=@n7.Result, False=Constant3Vector() {constant=<rgba 0 0 1 0>}, Value=@n8)
   World Position Offset
   Refraction
  ) {attributeSetTypes [
    #1
    #2
    #3
    #4
    #5
    #6
    #7
    #8
    #9
    #10
   ]}
  n21 = VectorParameter() {parameterName=BaseColor, defaultValue=<rgba 0.1086 0.1146 0.1146 1>}
  n22 = VectorParameter() {parameterName=BaseColor1, defaultValue=<rgba 1 0.9499 0.9167 1>}
  n23 = Multiply(A=Fresnel() {baseReflectFraction=0, exponent=12}, B=ScalarParameter() {defaultValue=3, parameterName=GlassThickness})
  n24 = Multiply(A=TextureCoordinate(), B=ScalarParameter() {defaultValue=1, parameterName=UV})
  n25 = TextureSample(UVs=@n24) {texture=/Game/Effects/Fort_Effects/Textures/Tile/T_Glass_Cracks_M}
  n26 = Saturate(None=Add(A=Saturate(None=@n23), B=@n25.R))
  n27 = Constant() {r=0.1237}
  n28 = Constant() {r=4}
  n29 = Multiply(
   A=MaterialFunctionCall(Radius=Constant() {r=0.5167}, Density=Constant() {r=4}) {materialFunction=&RadialGradientExponential}.RadialGradientExponential
   B=OneMinus(None=MaterialFunctionCall(Radius=@n27, Density=@n28) {materialFunction=&RadialGradientExponential}.RadialGradientExponential)
  )
  n30 = TextureSample(UVs=Multiply(A=TextureCoordinate(), B=ScalarParameter() {defaultValue=3, parameterName=DirtTile})) {samplerType=SAMPLERTYPE_Masks, texture=&T_DirtLeaks_M}
  SetMaterialAttributes!(
   MaterialAttributes
   World Position Offset=MaterialFunctionCall!(
    WSDeformationCenter=VectorParameter() {bUseCustomPrimitiveData=true, group=z_code, parameterName=WSDeformationCenter, primitiveDataIndex=20}.RGB
    Code_DestructionAmount=ScalarParameter() {group=z_code, parameterName=Code_DestructionAmount}
    Code_UseUvForDamage
    Code_UseHitEffects=StaticBoolParameter() {defaultValue=true, group=z_code, parameterName=Code_UseHitEffects}
    Code_Emissive
    WSDamageRadius=ScalarParameter() {bUseCustomPrimitiveData=true, defaultValue=1, group=z_code, parameterName=WSDamageRadius, primitiveDataIndex=26}
    CrackDamageShader
    Diffuse
    ScalarModDamageCracks
    ScalarOffsetforDamageCracks
    ExposeLerpFunctionForInvertingTheHitDistanceCalc
    InvertDamageFalloff
    CrackWorldPosition
    WorldPosition
    DistanceCullFadeAnimationOffset=ScalarParameter() {bUseCustomPrimitiveData=true, group=z_code, parameterName=DistanceCullFadeAnimationOffset, primitiveDataIndex=12}
   ) {materialFunction=/Game/Packages/jl_tests/Fortnite_HitEffects}.Hit And Shrink WPO
   Base Color=LinearInterpolate(
    A=@n1
    B=LinearInterpolate(A=@n21.RGB, B=@n22.RGB, Alpha=@n26)
    Alpha=MaterialFunctionCall(Transition Phase=@n29, Height Texture=@n30.B) {materialFunction=&HeightLerp}.Alpha
   )
   Metallic=ScalarParameter() {defaultValue=0.5, parameterName=Metallic}
   Specular=ScalarParameter() {defaultValue=0.5, parameterName=Spec}
   Roughness=LinearInterpolate(A=ScalarParameter() {defaultValue=0.1, parameterName=Rough}, B=ScalarParameter() {defaultValue=0.8, parameterName=Rough1}, Alpha=@n25.G) {constA=0.1, constB=0}
   Normal=MaterialFunctionCall(
    Normal=TextureSample(UVs=@n24) {samplerType=SAMPLERTYPE_Normal, texture=/Game/Effects/Fort_Effects/Textures/Tile/T_Glass_Cracks_N}.RGB
    Flatness=ScalarParameter() {defaultValue=0.25, parameterName=Flatten}
   ) {materialFunction=&FlattenNormal}.Result
  ) {attributeSetTypes [
    #9
    #1
    #2
    #3
    #4
    #8
   ]}
  ShadingPathSwitch(Default=@n18.RGB)
  TextureSample(UVs=@n24) {samplerType=SAMPLERTYPE_LinearGrayscale, texture=/Game/Environments/Papaya/Textures/shatter_kaleidoscope_lines}
 ]
}
